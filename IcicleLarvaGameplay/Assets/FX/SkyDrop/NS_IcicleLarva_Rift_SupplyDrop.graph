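# Niagara system for a sky rift supply-drop effect: cracks with a flickering flare and light, electric arcs, flying fragments, a glow and a large blue corona.
# Cracks uses five spawn groups to index per-layer arrays and visibility tags across mesh, sprite and light renderers, with curve-driven material parameters.
# Delayed bursts stagger glow, fragments and corona, while arcs run their own emitter lifecycle.
# User.FX_Scale scales much of it; RisingProcess and ScaleFX appear unused.
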NiagaraSystem /IcicleLarvaGameplay/Assets/FX/SkyDrop/NS_IcicleLarva_Rift_SupplyDrop
dependencies {
 usedDataInterfaces [
  &NiagaraDataInterfaceCurve
  &NiagaraDataInterfaceVectorCurve
  &NiagaraDataInterfaceArrayFloat
  &NiagaraDataInterfaceArrayColor
  &NiagaraDataInterfaceArrayFloat3
 ]
 usedDynamicInputs [&ConvertVectorToPosition]
 usedModules [&SystemState]
}
emitters {
 Corona {
  emitterUpdateScript [
   &EmitterState()
   &SpawnBurst_Instantaneous(Spawn Time=2, Use Spawn Probability=false, UseLoopCountLimit=false)
  ]
  particleSpawnScript [
   &InitializeParticle(Lifetime=1.2, Sprite Size=<25000 25000>, Sprite Size Mode=::NewEnumerator4, UsePositionOffset=false, Write Material Random=true)
   &Color(Color=&MakeLinearColorFromVectorAndFloat("Vector (RGB)"=<0.1344 0.4283 1>, "Float (Alpha)"=&ClampFloat(Float=1)))
  ]
  particleUpdateScript [
   &ParticleState()
   &ScaleSpriteSize(
    Scale Sprite Size Mode=::NewEnumerator0
    Uniform Scale Factor=&FloatFromCurve(
     FloatCurve=DataInterface() {propertyValues {Curve {keys [
         {arriveTangent=2.8298, interpMode=RCIM_Cubic, leaveTangent=2.8298, tangentMode=RCTM_User, time=0, value=0.1}
         {interpMode=RCIM_Cubic, leaveTangent=0.2191, time=1, value=1}
        ]}}}
     Scale Curve=&Multiply_Float(A=$User.FX_Scale, B=0.25)
    )
   )
   &ScaleColor(
    Scale Alpha=&FloatFromCurve(FloatCurve=DataInterface() {propertyValues {Curve {keys [<at 0 0 RCIM_Cubic>,<at 0.25 1 RCIM_Cubic>,<at 1 0>]}}})
    Scale RGB=<2 2 2>
    ScaleA=true
    ScaleRGB=true
   )
   &DynamicMaterialParameters(
    Index 0 Param 1=&FloatFromCurve(FloatCurve=DataInterface() {propertyValues {Curve {keys [{arriveTangent=-2.4947, interpMode=RCIM_Cubic, time=0, value=1.5}, <at 1 0.2 RCIM_Cubic>]}}})
    Index 0 Param 2=0.4
    Index 0 Param 3=0.6
    Index 0 Param 4=1
    Module.Index0Param0Write=true
    Module.Index0Param1Write=true
    Module.Index0Param2Write=true
    Module.Index0Param3Write=true
   )
   &CameraOffset(Camera Offset Amount=500)
  ]
  rendererClasses [&NiagaraSpriteRendererProperties]
  renderers [{rendererClass=&NiagaraSpriteRendererProperties}]
 }
 Cracks {
  emitterUpdateScript [
   &EmitterState()
   &SpawnBurst_Instantaneous(Spawn Count=5, Use Spawn Probability=false, UseLoopCountLimit=true)
  ]
  particleSpawnScript [
   &FN_SetSpawnGroup(Number of Spawn Groups=5)
   /SetVariables(Particles.VisibilityTag=$Particles.SpawnGroup) {isSetParameters=true}
   &InitializeParticle(
    Color=&SelectLinearColorFromArray(
     Array Sampling Mode=::NewEnumerator1
     Direct Array Index=$Particles.SpawnGroup
     Color Selection Array=DataInterface() {propertyValues {ColorData [<rgba 1 1 1 1>, <rgba 1 1 1 1>, <rgba 1 1 1 1>, <rgba 0.1637 1.4491 2 1>]}}
     Random Seed=?
    )
    Lifetime Max=4
    Lifetime Min:NiagaraFloat=&SelectFloatFromArray(Direct Array Index=$Particles.SpawnGroup, Float Selection Array=DataInterface() {propertyValues {FloatData [3,3,3,3,4]}})
    Lifetime Mode=::NewEnumerator1
    Mesh Scale Mode=::NewEnumerator1
    Mesh Uniform Scale Max:NiagaraFloat=&SelectFloatFromArray(Direct Array Index=$Particles.SpawnGroup, Float Selection Array=DataInterface() {propertyValues {FloatData [75,125,100]}})
    Mesh Uniform Scale Min:NiagaraFloat=&SelectFloatFromArray(Direct Array Index=$Particles.SpawnGroup, Float Selection Array=DataInterface() {propertyValues {FloatData [50,100,100]}})
    Sprite Rotation Mode=::NewEnumerator2
    Sprite Size Mode=::NewEnumerator3
    Uniform Sprite Size=37500
    UsePositionOffset=false
    Write Material Random=true
   )
   &InitialMeshOrientation(
    Mesh Orientation Mode=::NewEnumerator2
    Orientation Vector=&SelectVectorFromArray(Direct Array Index=$Particles.SpawnGroup, Vector Selection Array=DataInterface() {propertyValues {FloatData [<1 0 0>, <1 0 0>, <0 0 -1>]}})
    Rotation=&RandomRangeVector(
     Minimum=<0 0 0>
     Maximum=&SelectVectorFromArray(Direct Array Index=$Particles.SpawnGroup, Vector Selection Array=DataInterface() {propertyValues {FloatData [<1 0 0>, <1 0 0>, <0 0 1>]}})
    )
    Use Rotation Vector=true
   )
  ]
  particleUpdateScript [
   &ParticleState()
   &ScaleMeshSize(Scale Factor=&VectorFromFloat(Value=$User.FX_Scale))
   &ScaleSpriteSize(
    Uniform Curve Scale=$User.FX_Scale
    Uniform Curve Sprite Scale=DataInterface() {propertyValues {Curve {keys [
        {interpMode=RCIM_Cubic, tangentMode=RCTM_User, time=0, value=0.3}
        {arriveTangent=0.7525, interpMode=RCIM_Cubic, leaveTangent=0.7525, tangentMode=RCTM_User, time=0.8, value=0.45}
        {interpMode=RCIM_Cubic, tangentMode=RCTM_User, time=0.9, value=1}
       ]}}}
   )
   &DynamicMaterialParameters(
    Index 0 Param 1=&MaskFloatBySpawnGroup(
     Masked Float=&FloatFromCurve(FloatCurve=DataInterface() {propertyValues {Curve {keys [<at 0 0>,<at 0.05 0.2>,<at 0.225 0.535>,<at 0.6 0.535>,<at 0.625 0.6>,<at 0.9 0.3>,<at 0.925 0.175>,<at 1 0>]}}})
     Passthrough Float=&FloatFromCurve(
      FloatCurve=DataInterface() {propertyValues {Curve {keys [<at 0 0>,<at 0.05 0.2>,<at 0.225 0.535>,<at 0.6 0.535>,<at 0.625 0.6>,<at 0.9 0.3>,<at 0.925 0.175>,<at 1 0>]}}}
      Scale Curve=2
     )
    )
    Index 0 Param 2=&MaskFloatBySpawnGroup(
     Masked Float=&FloatFromCurve(
      FloatCurve=DataInterface() {propertyValues {Curve {keys [
          {arriveTangent=3.4784, interpMode=RCIM_Cubic, leaveTangent=3.4784, tangentMode=RCTM_User, time=0.2, value=0}
          <at 0.8 1.05 RCIM_Cubic>
          {arriveTangent=-12.1443, interpMode=RCIM_Cubic, leaveTangent=-12.1443, tangentMode=RCTM_User, time=0.9993, value=0}
         ]}}}
      Scale Curve=17
     )
     Passthrough Float=&FloatFromCurve(
      FloatCurve=DataInterface() {propertyValues {Curve {keys [
          {arriveTangent=3.4784, interpMode=RCIM_Cubic, leaveTangent=3.4784, tangentMode=RCTM_User, time=0.2, value=0}
          <at 0.8 1.05 RCIM_Cubic>
          {arriveTangent=-12.1443, interpMode=RCIM_Cubic, leaveTangent=-12.1443, tangentMode=RCTM_User, time=0.9993, value=0}
         ]}}}
      Scale Curve=25
     )
    )
    Index 0 Param 3=1
    Index 0 Param 4=1
    Module.Param0WriteEnabled=true
    Module.Param1WriteEnabled=true
    Module.Param2WriteEnabled=true
    Module.Param3WriteEnabled=true
   )
   /SetVariables(Particles.DMP_Flare:Vector4f=<1 1 1 1>, Particles.DMP_Shatter:Vector4f=<0 0 0 0.05>) {isSetParameters=true}
   /SetVariables(
    Particles.CL_Flare:LinearColor=&MultiplyLinearColor(
     Linear Color=$Particles.Initial.Color
     Scale Factor=&MakeFloat4FromVectorAndFloat(
      "Vector (RGB)"=<0.5 0.5 0.5>
      "Float (Alpha)"=&FloatFromCurve(FloatCurve=DataInterface() {propertyValues {Curve {keys [
           <at 0 0 RCIM_Cubic>
           {arriveTangent=0.289, interpMode=RCIM_Cubic, leaveTangent=0.289, time=0.4918, value=0.248}
           {arriveTangent=3.4072, interpMode=RCIM_Cubic, leaveTangent=3.4072, time=0.5041, value=0.1457}
           {arriveTangent=-2.7103, interpMode=RCIM_Cubic, leaveTangent=-2.7103, time=0.5232, value=0.3549}
           {arriveTangent=2.94, interpMode=RCIM_Cubic, leaveTangent=2.94, time=0.5374, value=0.0554}
           {arriveTangent=0.8123, interpMode=RCIM_Cubic, leaveTangent=0.8123, time=0.5612, value=0.4666}
           {arriveTangent=3.9277, interpMode=RCIM_Cubic, leaveTangent=3.9277, time=0.5922, value=0.0999}
           {arriveTangent=3.855, interpMode=RCIM_Cubic, leaveTangent=3.855, time=0.6114, value=0.6636}
           {arriveTangent=-6.193, interpMode=RCIM_Cubic, leaveTangent=-6.193, time=0.6303, value=0.2469}
           {arriveTangent=1.1094, interpMode=RCIM_Cubic, leaveTangent=1.1094, time=0.6459, value=0.4498}
           {arriveTangent=3.878, interpMode=RCIM_Cubic, leaveTangent=3.878, time=0.6893, value=0.3124}
           {arriveTangent=0.3841, interpMode=RCIM_Cubic, leaveTangent=0.3841, time=0.6979, value=0.6518}
           {arriveTangent=1.6385, interpMode=RCIM_Cubic, leaveTangent=1.6385, time=0.7238, value=0.3256}
           {arriveTangent=1.2094, interpMode=RCIM_Cubic, leaveTangent=1.2094, time=0.7388, value=0.7187}
           {arriveTangent=5.1593, interpMode=RCIM_Cubic, leaveTangent=5.1593, time=0.769, value=0.3803}
           {arriveTangent=2.2949, interpMode=RCIM_Cubic, leaveTangent=2.2949, time=0.7839, value=0.9515}
           {arriveTangent=-4.6065, interpMode=RCIM_Cubic, leaveTangent=-4.6065, time=0.815, value=0.4859}
           {arriveTangent=9.5468, interpMode=RCIM_Cubic, leaveTangent=9.5468, time=0.8665, value=0.5711}
           {arriveTangent=0.1948, interpMode=RCIM_Cubic, leaveTangent=0.1948, time=0.8707, value=1.017}
           {arriveTangent=-7.1532, interpMode=RCIM_Cubic, leaveTangent=-7.1532, time=0.8934, value=0.5764}
           {arriveTangent=15.7037, interpMode=RCIM_Cubic, leaveTangent=15.7037, time=0.925, value=0.6286}
           {arriveTangent=-2.0814, interpMode=RCIM_Cubic, leaveTangent=-2.0814, time=0.9294, value=1.1414}
           {arriveTangent=-16.7837, interpMode=RCIM_Cubic, leaveTangent=-16.7837, time=0.9479, value=0.5809}
           {arriveTangent=31.4448, interpMode=RCIM_Cubic, leaveTangent=31.4448, time=0.9612, value=0.6074}
           {arriveTangent=2.7173, interpMode=RCIM_Cubic, leaveTangent=2.7173, time=0.9678, value=1.2074}
           {arriveTangent=-6.4425, interpMode=RCIM_Cubic, leaveTangent=-6.4425, time=0.9866, value=0.6763}
           <at 1 1>
          ]}}})
     )
    )
    Particles.CL_Shatter:LinearColor=&MultiplyLinearColor(
     Linear Color=$Particles.Initial.Color
     Scale Factor=&MakeFloat4FromVectorAndFloat(
      "Vector (RGB)"=<1 1 1>
      "Float (Alpha)"=&FloatFromCurve(FloatCurve=DataInterface() {propertyValues {Curve {keys [<at 0 0>,<at 0.2 0>,<at 0.225 0.075>,<at 0.6 0.1>,<at 0.625 0.175>,<at 0.8 0.2125>,<at 0.9 0.1625>,<at 0.98 0>]}}})
     )
    )
   ) {isSetParameters=true}
   &CameraOffset(Camera Offset Amount=&SelectFloatFromArray(Direct Array Index=$Particles.SpawnGroup, Float Selection Array=DataInterface() {propertyValues {FloatData [100,0,0,0]}}))
   &FN_Light_Attributes(
    Color=<rgba 0.1637 1.2404 2 1>
    Create Color Attribute=true
    Create Exponent Attribute=true
    Create Radius Attribute=true
    Exponent=4
    Radius=&Lerp_Float(
     A=500
     B=900
     Alpha=&FloatFromCurve(FloatCurve=DataInterface() {propertyValues {Curve {keys [<at 0 0>, {interpMode=RCIM_Cubic, tangentMode=RCTM_User, time=0.5, value=0.2}, <at 0.9 1>]}}})
    )
    Scale Color=8
   )
  ]
  rendererClasses [
   &NiagaraMeshRendererProperties
   &NiagaraSpriteRendererProperties
   &NiagaraLightRendererProperties
  ]
  renderers [
   {rendererClass=&NiagaraMeshRendererProperties, rendererIndex=0}
   {rendererClass=&NiagaraMeshRendererProperties, rendererIndex=1}
   {rendererClass=&NiagaraMeshRendererProperties, rendererIndex=2}
   {rendererClass=&NiagaraSpriteRendererProperties, rendererIndex=3}
   {rendererClass=&NiagaraLightRendererProperties, rendererIndex=4}
  ]
 }
 ElectricArcsResidual {
  emitterUpdateScript [
   &EmitterState(Life Cycle Mode=::NewEnumerator1, Loop Behavior=::NewEnumerator1, Loop Delay=1.5, Loop Duration=0.75, UseLoopDelay=true)
   &SpawnBurst_Instantaneous(Spawn Count=6, Use Spawn Probability=false, UseLoopCountLimit=false)
   &SpawnRate(
    SpawnRate=&Multiply_Float(
     A=30
     B=&FloatFromCurve(
      FloatCurve=DataInterface() {propertyValues {Curve {keys [{arriveTangent=-2.9237, interpMode=RCIM_Cubic, leaveTangent=-2.9237, tangentMode=RCTM_User, time=0, value=1}, <at 0.75 0>]}}}
      CurveIndex=$Emitter.NormalizedLoopAge
     )
    )
    Use Spawn Probability=false
   )
  ]
  particleSpawnScript [
   &InitializeParticle(
    Lifetime=&RandomRangeFloat(Minimum=0.1, Maximum=0.25)
    Mesh Scale=&RandomRangeVector(Minimum=<50 50 50>, Maximum=<120 90 90>)
    Mesh Scale Mode=::NewEnumerator4
    UsePositionOffset=false
    Write Material Random=true
   )
   &InitialMeshOrientation(
    Mesh Orientation Mode=::NewEnumerator7
    Rotation=&RandomRangeVector(Minimum=<0 0 0>, Maximum=<1 0 0>)
    Use Rotation Vector=true
   )
   &ShapeLocation(
    Cylinder Height=20
    Cylinder Radius=250
    MaskSpawnGroup=false
    Override Seed=false
    Rotation Axis=<0 0 1>
    Rotation Mode=::NewEnumerator1
    Shape Primitive=::NewEnumerator1
    Surface Only=false
    UseOwnerScale=true
   )
   &AddVelocity(Velocity=&Multiply_VectorByFloat(Vector=&Multiply_Vector(A=$Particles.ShapeLocation.ShapeVector, B=<1 1 0>), Float=5))
   &Color(Color=&MakeLinearColorFromVectorAndFloat(
     "Vector (RGB)"=&RandomRangeVector(Minimum=<3.9751 0.4547 4>, Maximum=<0 6 4.9028>)
     "Float (Alpha)"=&ClampFloat(Float=1)
    ))
  ]
  particleUpdateScript [
   &ParticleState()
   &ScaleMeshSize(Scale Factor=&VectorFromCurve(
     VectorCurve=DataInterface() {propertyValues {
       XCurve {keys [<at 0.0033 0.5 RCIM_Constant>,<at 0.2054 0.7619 RCIM_Constant>,<at 0.4101 0.8575 RCIM_Constant>,<at 0.6967 1>]}
       YCurve {keys [<at 0.0033 0.6004 RCIM_Constant>,<at 0.2054 0.6716 RCIM_Constant>,<at 0.4101 0.3646 RCIM_Constant>,<at 0.6967 0.2468>]}
       ZCurve {keys [<at 0.0033 0.7767 RCIM_Constant>,<at 0.2054 0.5215 RCIM_Constant>,<at 0.4101 0.4457 RCIM_Constant>,<at 0.6967 0.6282>]}
      }}
     Scale Curve=&VectorFromFloat(Value=$User.FX_Scale)
    ))
   &DynamicMaterialParameters(
    Index 0 Param 1=&RandomRangeFloat(Minimum=2, Maximum=3)
    Index 0 Param 2=&FloatFromCurve(FloatCurve=DataInterface() {propertyValues {Curve {keys [{arriveTangent=-1.8894, interpMode=RCIM_Cubic, time=0, value=1}, <at 1 -0.25>]}}})
    Index 0 Param 3=&RandomRangeFloat(Minimum=0.2, Maximum=2)
    Index 0 Param 4=&RandomRangeFloat(Minimum=0, Maximum=1)
    Module.Index0Param0Write=true
    Module.Index0Param1Write=true
    Module.Index0Param2Write=true
    Module.Index0Param3Write=true
   ) {module=DynamicMaterialParameters001}
   &ScaleColor(
    Scale RGB=&VectorFromFloat(Value=&FloatFromCurve(FloatCurve=DataInterface() {propertyValues {Curve {keys [<at 0 50 RCIM_Cubic>, {arriveTangent=-45, interpMode=RCIM_Cubic, leaveTangent=-45, time=0.5, value=30}, <at 1 5>]}}}))
    ScaleA=true
    ScaleRGB=true
   )
   &SolveForcesAndVelocity(Clamp Velocity=false, Limit Acceleration=false)
  ]
  rendererClasses [&NiagaraMeshRendererProperties]
  renderers [{rendererClass=&NiagaraMeshRendererProperties, rendererIndex=0}]
 }
 Frags {
  emitterUpdateScript [
   &EmitterState(Life Cycle Mode=::NewEnumerator1, Loop Behavior=::NewEnumerator1, Loop Duration=1, UseLoopDelay=true)
   &SpawnBurst_Instantaneous(Spawn Count=12, Spawn Group=2, Spawn Time=0.5, Use Spawn Probability=false, UseLoopCountLimit=false) {module=SpawnBurst_Instantaneous002}
  ]
  particleSpawnScript [
   /SetVariables(Particles.MeshIndex=$Emitter.SpawnGroup) {isSetParameters=true}
   &InitializeParticle(
    Color=<rgba 5 5 5 1>
    Lifetime=&RandomRangeFloat(Minimum=1, Maximum=1.5)
    Mesh Scale=&RandomRangeVector(Minimum=<5 5 5>, Maximum=<25 12 12>)
    Mesh Scale Mode=::NewEnumerator4
    UsePositionOffset=false
    Write Material Random=true
   )
   &ShapeLocation(
    MaskSpawnGroup=false
    Non Uniform Scale=&VectorFromFloat(Value=$User.FX_Scale)
    Override Seed=false
    Sphere Radius=&RandomRangeFloat(Minimum=30, Maximum=40)
    UseOwnerScale=true
   )
   /SetVariables(Particles.MeshOrientation=&MakeQuatFromAxisAngle(Axis=&RandomRangeVector(Minimum=<0 0 0>, Maximum=<1 1 1>), AngleInDegrees=90)) {isSetParameters=true}
   &AddVelocity(
    Module.Use Falloff Distance=false
    Override Seed=false
    Velocity Mode=::NewEnumerator1
    Velocity Speed=&RandomRangeFloat(Minimum=10000, Maximum=20000)
   )
   /SetVariables(Particles.SpawnRotationVector=&RandomRangeVector(Minimum=<0.1 1.5 -0.1>, Maximum=<0.4 2 0.1>)) {isSetParameters=true}
   &AddRotationalVelocity(Rotation Rate=$StackContext.SpawnRotationVector)
  ]
  particleUpdateScript [
   &ParticleState()
   &Drag(
    Drag=&RandomRangeFloat(Minimum=2.5, Maximum=5, Evaluation Type=::NewEnumerator1)
    Use Linear Drag=true
    Use Rotational Drag=true
   )
   &SolveForcesAndVelocity(Clamp Velocity=false, Limit Acceleration=false)
   &ScaleMeshSize(Scale Factor=&VectorFromCurve(
     VectorCurve=DataInterface() {propertyValues {
       XCurve {keys [<at 0.2 1 RCIM_Cubic>, {arriveTangent=-1.25, interpMode=RCIM_Cubic, leaveTangent=-1.25, time=0.7932, value=0.7509}, <at 1 0>]}
       YCurve {keys [<at 0.2 1 RCIM_Cubic>, {arriveTangent=-0.25, interpMode=RCIM_Cubic, leaveTangent=-0.25, time=0.7932, value=0.85}, <at 1 0.8>]}
       ZCurve {keys [<at 0.2 1 RCIM_Cubic>,<at 0.7932 1 RCIM_Cubic>,<at 1 1>]}
      }}
     Scale Curve=<2 2.1627 2>
    ))
   &ScaleColor(
    Scale Alpha=&FloatFromCurve(FloatCurve=DataInterface() {propertyValues {Curve {keys [<at 0.7 1 RCIM_Cubic>,<at 1 0>]}}})
    Scale RGB=&VectorFromFloat(Value=&FloatFromCurve(FloatCurve=DataInterface() {propertyValues {Curve {keys [<at 0.1 25 RCIM_Cubic>, {arriveTangent=-26.6667, interpMode=RCIM_Cubic, leaveTangent=-26.6667, time=0.1961, value=7.5192}, <at 1 1>]}}}))
    ScaleA=true
    ScaleRGB=true
   )
   &CameraOffset(Camera Offset Amount=-50)
  ]
  rendererClasses [&NiagaraMeshRendererProperties]
  renderers [{rendererClass=&NiagaraMeshRendererProperties, rendererIndex=0}]
 }
 Glow {
  emitterUpdateScript [
   &EmitterState()
   &SpawnBurst_Instantaneous(Spawn Time=0.5, Use Spawn Probability=false, UseLoopCountLimit=false)
  ]
  particleSpawnScript [
   &InitializeParticle(
    Color=<rgba 0.0441 0.3919 1 1>
    Sprite Rotation Mode=::NewEnumerator2
    Sprite Size=<975 975>
    Sprite Size Mode=::NewEnumerator4
    UsePositionOffset=false
    Write Material Random=true
   ) {module=InitializeParticle001}
   &DynamicMaterialParameters(
    Index 0 Param 1=1
    Index 0 Param 3=1
    Index 0 Param 4=1
    Index 1 Param 1=200
    Module.Param0WriteEnabled=true
    Module.Param0WriteEnabled1=true
    Module.Param1WriteEnabled=true
    Module.Param1WriteEnabled1=false
    Module.Param2WriteEnabled=true
    Module.Param2WriteEnabled1=false
    Module.Param3WriteEnabled=true
    Module.Param3WriteEnabled1=false
    Write Parameter Index 1=true
   )
  ]
  particleUpdateScript [
   &ParticleState()
   &ScaleSpriteSize(Scale Factor=&Multiply_Vector2D_ByFloat(
     Float=30
     Vector2D=&Vector2DFromFloat(Value=&Multiply_Float(
       A=&FloatFromCurve(FloatCurve=DataInterface() {propertyValues {Curve {keys [<at 0 0>, {arriveTangent=-3.1325, leaveTangent=-3.1325, time=0.2, value=1}, <at 1 0>]}}})
       B=1
      ))
    ))
   &ScaleColor(
    Scale Alpha=&FloatFromCurve(FloatCurve=DataInterface() {propertyValues {Curve {keys [{interpMode=RCIM_Cubic, tangentMode=RCTM_User, time=0, value=1}, {interpMode=RCIM_Cubic, tangentMode=RCTM_User, time=1, value=0}]}}})
    Scale RGB=&VectorFromFloat(Value=&FloatFromCurve(FloatCurve=DataInterface() {propertyValues {Curve {keys [<at 0 1>,<at 1 0>]}}}))
    ScaleA=true
    ScaleRGB=true
   )
   &CameraOffset(Camera Offset Amount=200)
  ]
  rendererClasses [&NiagaraSpriteRendererProperties]
  renderers [{rendererClass=&NiagaraSpriteRendererProperties}]
 }
}
summary {
 systemName=NS_IcicleLarva_Rift_SupplyDrop
 emitters [
  {emitterName=ElectricArcsResidual, rendererClasses [&NiagaraMeshRendererProperties]}
  {emitterName=Corona, rendererClasses [&NiagaraSpriteRendererProperties]}
  {emitterName=Frags, rendererClasses [&NiagaraMeshRendererProperties]}
  {emitterName=Glow, rendererClasses [&NiagaraSpriteRendererProperties]}
  {
   emitterName=Cracks
   rendererClasses [
    &NiagaraMeshRendererProperties
    &NiagaraSpriteRendererProperties
    &NiagaraLightRendererProperties
   ]
  }
 ]
 userVariables [
  User.FX_Scale {type=NiagaraFloat, defaultValue=2, description=""}
  User.RisingProcess {type=NiagaraFloat, defaultValue=0, description=""}
  User.ScaleFX {type=NiagaraFloat, defaultValue=100, description=""}
 ]
}
